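# Editor utility that stops dedicated servers stripping anim data from sequences used by Fortnite Mantis montages.
# Entry points process selected data assets, montages or sequences, or every FortMantisData via the Asset Registry.
# Data assets pass each technique's main montage, ignoring DynamicMontages, to a FortMantis notify check that yields its sequences.
# A shared function sets DoNotStrip where needed, then checks out and saves; the "Unsafe" sequence entry skips the notify check.
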EditorUtilityBlueprint /Mantis/FortMantisAnimSequenceValidator
graphs [
 Checkout And Save Anim Sequences {nodes [
   n1 = Checkout And Save Anim Sequences!() -> (then, AnimSequence:Object Reference) {is=title K2Node_FunctionEntry}
   > n2 = Cast To AnimSequence!(execute, Object=@n1) -> (then, CastFailed, AsAnim Sequence:Anim Sequence Object Reference)
   > n3 = GetStripDataFromServerForAnimSequence!(execute, AnimSequence:Anim Sequence Object Reference=@n2) -> (then, ReturnValue:EStripAnimDataOnDedicatedServerSettings Enum=UseProjectSetting)
   > Branch(Condition="Not Equal (Enum)"!(A:EStripAnimDataOnDedicatedServerSettings Enum=@n3, B:EStripAnimDataOnDedicatedServerSettings Enum=DoNotStripAnimDataOnDedicatedServer) -> (ReturnValue:bool))
   > SetStripDataFromServerForAnimSequence!(
    execute
    AnimSequence:Anim Sequence Object Reference=@n2
    StripDataOnDedicatedServerSettings:EStripAnimDataOnDedicatedServerSettings Enum=DoNotStripAnimDataOnDedicatedServer
   ) -> (then)
   > n4 = CheckoutLoadedAsset!(execute, AssetToCheckout:Object Reference=@n2) -> (then, ReturnValue:bool)
   > n5 = Branch(Condition=@n4)
   > SaveLoadedAsset(AssetToSave=@n2, bOnlyIfIsDirty=false)
   PrintString(
    execute=@n5.else
    InString=Append!(A="Could not checkout ", B=GetDisplayName(Object=@n2), C:String=" because it was locked by another user.") -> (ReturnValue)
   )
  ]}
 EventGraph {links [], nodes []}
 ValidateMantisAnimMontages {nodes [
   ValidateMantisAnimMontages() {is=title K2Node_FunctionEntry}
   > n1 = GetSelectedAssetsOfClass(AssetClass=/Script/Engine.AnimMontage)
   > n2 = For Each Loop(Array:Array of Object References=@n1) -> (Array Element:Object Reference)
   > n3 = Cast To AnimMontage!(execute, Object=@n2) -> (then, CastFailed, AsAnim Montage:Anim Montage Object Reference) {is=title K2Node_DynamicCast}
   > ValidateSingleMantisAnimMontage!(execute, self, AnimMontages:Anim Montage Object Reference=@n3) -> (then)
  ]}
 "ValidateMantisAnimSequences (Unsafe)" {nodes [
   "ValidateMantisAnimSequences (Unsafe)"() {is=title K2Node_FunctionEntry}
   > n1 = GetSelectedAssetsOfClass(AssetClass=/Script/Engine.AnimSequence)
   > n2 = For Each Loop(Array=@n1)
   > Checkout And Save Anim Sequences!(execute, self, AnimSequence:Object Reference=@n2) -> (then) {is=title K2Node_CallFunction}
  ]}
 ValidateMantisData {nodes [
   ValidateMantisData() {is=title K2Node_FunctionEntry}
   > n1 = GetSelectedAssetsOfClass(AssetClass=/Script/MantisRuntime.FortMantisData)
   > n2 = For Each Loop(Array=@n1)
   > ValidateSingleMantisDataAsset!(
    execute
    self
    FortMantisDataAsset:Fort Mantis Data Object Reference=Cast To FortMantisData!(Object=@n2) -> (AsFort Mantis Data:Fort Mantis Data Object Reference, bSuccess:bool=~) {is=title K2Node_DynamicCast}
   ) -> (then)
  ]}
 ValidateProjectWideMantisAnimSequences {nodes [
   ValidateProjectWideMantisAnimSequences() {is=title K2Node_FunctionEntry}
   n1 = Get FortMantisData Ref() -> (FortMantisData Ref:Fort Mantis Data Object Reference)
   > n2 = GetAssetsByClass!(
    execute
    self:Asset Registry Interface=GetAssetRegistry()
    ClassPathName:Top Level Asset Path Structure="Get Class Path Name (Top Level Asset Path)"!(Class:Object Class Reference=Get Class(Object=@n1) -> (ReturnValue:Fort Mantis Data Class Reference)) -> (ReturnValue:Top Level Asset Path Structure) {is=title K2Node_CallFunction}
    bSearchSubClasses:bool=true
   ) -> (then, OutAssetData:Array of Asset Data Structures, ReturnValue:bool)
   > Branch(Condition=@n2.ReturnValue)
   > n3 = For Each Loop(Array:Array of Asset Data Structures=@n2) -> (Array Element:Asset Data Structure)
   > ValidateSingleMantisDataAsset(FortMantisDataAsset=Cast To FortMantisData(Object=GetAsset(InAssetData=@n3)) -> (bSuccess=~))
   n4 = Length(TargetArray:Array of Asset Data Structures=@n2)
   PrintString(execute=@n3.Completed, InString=Append(A="Mantis Anim Validation Ran For: ", B="To String (Integer)"(InInt=@n4), C=" Mantis Data Assets"))
  ]}
 ValidateSingleMantisAnimMontage {nodes [
   n1 = ValidateSingleMantisAnimMontage!() -> (then, AnimMontages:Anim Montage Object Reference) {is=title K2Node_FunctionEntry}
   > Branch(Condition=Contains Item(
     TargetArray=GetAnimationNotifyEventNames!(AnimationSequenceBase:Anim Sequence Base Object Reference=@n1) -> (EventNames:Array of Names)
     ItemToFind=FortMantis
    ))
   > n2 = GetSequencesFromMontage!(execute, Montage:Anim Montage Object Reference=@n1) -> (then, ReturnValue:Array of Anim Sequence Object References)
   > n3 = For Each Loop(Array:Array of Anim Sequence Object References=@n2) -> (Array Element:Anim Sequence Object Reference)
   > Checkout And Save Anim Sequences(AnimSequence=@n3)
  ]}
 ValidateSingleMantisDataAsset {nodes [
   n1 = ValidateSingleMantisDataAsset!() -> (then, FortMantisDataAsset:Fort Mantis Data Object Reference) {is=title K2Node_FunctionEntry}
   > n2 = Cast To FortMantisData!(execute:Exec, Object=@n1) -> (then:Exec, CastFailed:Exec, AsFort Mantis Data)
   > n3 = For Each Loop!(
    Exec
    Array:Array of Fort Mantis Technique Data Structures=Get Techniques!(self:Fort Mantis Data Object Reference=@n2) -> (Techniques:Array of Fort Mantis Technique Data Structures)
   ) -> (
    LoopBody
    Array Element_Name:Name
    Array Element_Montage:Anim Montage Object Reference
    Array Element_DynamicMontages:Map of Names to Anim Montage Object References
    Array Index
    Completed
   )
   > ValidateSingleMantisAnimMontage(AnimMontages=@n3.Array Element_Montage) {is=title K2Node_CallFunction}
  ]}
]
variables []
